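# Landmass brush material that composites one water body into the combined velocity, height and depth texture.
# A cached distance field supplies velocity, height and depth, with height localized to landscape Z and depth normalized and inverted.
# A dilated shape mask blends these over existing values.
# A custom neighbor search gates that dilation, which also depends on a height comparison, to texels lacking nearby valid depth in the existing texture.
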Material /Water/Materials/Brushes/CompositeWaterBodyTexture
graph {
 properties {
  bAllowNegativeEmissiveColor=true
  bUseTranslucencyVertexFog=false
  blendMode=BLEND_AlphaComposite
  floatPrecisionMode=MFPM_Full
  pixelDepthOffsetMode=PDOM_Legacy
  shadingModel=MSM_Unlit
  usedShadingModels=Unlit
 }
 nodes [
  n1 = TextureSampleParameter2D() {desc="RG=Velocity, B=Water Height,  A=WaterDepth", parameterName=CombinedVelocityAndHeight, texture=&Black}
  n2 = MaterialFunctionCall!(Cached DF, UV Offset) {materialFunction=/Landmass/Landscape/BlueprintBrushes/MF/ReadCachedDistanceField}
  n3 = NamedRerouteUsage()
  n4 = ScalarParameter() {parameterName=ZOffset}
  n5 = NamedRerouteDeclaration(None=Subtract(
    A=@n2.Water Height
    B=MaterialFunctionCall!(World Z=@n4) {materialFunction=/Landmass/Landscape/BlueprintBrushes/MF/LocalizeLandscapeZValue}.Scale Only
   )) {name=Water Z Height, nodeColor=<rgba 1 0 0.1875 1>}
  n6 = LinearInterpolate(A=Min(A=Subtract(A=@n3) {constB=64}, B=@n5), B=@n5, Alpha=NamedRerouteUsage())
  n7 = Saturate(None=MaterialFunctionCall() {materialFunction=&ReadCachedDistanceField}.World Distance)
  n8 = NamedRerouteUsage()
  n9 = TextureSampleParameter2D() {desc="RG=Velocity, B=Water Height,  A=WaterDepth", parameterName=CombinedVelocityAndHeight, texture=&Black}
  n10 = Constant() {r=1}
  n11 = Constant()
  MaterialExpressionTextureObjectParameter_0 = TextureObjectParameter() {parameterName=CombinedVelocityAndHeight, texture=&Black}
  MaterialExpressionTextureCoordinate_0 = TextureCoordinate()
  n12 = CollectionParameter() {collection=/Landmass/Landscape/BlueprintBrushes/MPC/MPC_Landscape, parameterName=RTResX}
  n13 = CollectionParameter() {collection=&MPC_Landscape, parameterName=RTResY}
  MaterialExpressionMultiply_10 = Multiply(A=Divide(B=AppendVector(A=@n12, B=@n13)), B=ScalarParameter() {defaultValue=2, parameterName=Texel Scale})
  MaterialExpressionScalarParameter_8 = ScalarParameter() {defaultValue=3, parameterName=Radius}
  n14 = Custom!(
   Tex=@MaterialExpressionTextureObjectParameter_0
   UV=@MaterialExpressionTextureCoordinate_0
   TexelSize=@MaterialExpressionMultiply_10
   Radius=@MaterialExpressionScalarParameter_8
  ) {
   code="float2 accum = 0;\r\n\r\nfloat2 curUV = UV;\r\nfloat neighbors = 0;\r\n\r\nfor (int i = -Radius; i <= Radius; i++)\r\n{\r\n\r\n\tfor(int j=-Radius; j <= Radius; j++)\r\n\t{\r\n\t\t\r\n\t\tcurUV = UV + float2(TexelSize.x * (float)i, 0) + float2(0, TexelSize.y * (float)j);\r\n\t\tfloat4 cursample = Tex.SampleLevel(TexSampler, curUV, 0);\r\n\t\t\r\n\t\tfloat depthmask = 1 - floor(cursample.a);\r\n\t\t\r\n\t\tif (depthmask == 1.0)\r\n\t\t{\r\n\t\t\taccum += cursample.xy;\r\n\t\t\tneighbors += 1;\r\n\t\t\r\n\t\t}\r\n\t\t\r\n\t}\r\n\t\r\n}\r\n\r\nif(neighbors > 0)\r\n{\r\n\taccum.xy /= neighbors;\r\n}\r\n\r\nreturn float3( accum.xy, saturate(neighbors) );"
   description=Neighbor Search
   inputs [
    {inputName=Tex, input {expression=&self:MaterialExpressionTextureObjectParameter_0}}
    {inputName=UV, input {expression=&self:MaterialExpressionTextureCoordinate_0}}
    {inputName=TexelSize, input {expression=&self:MaterialExpressionMultiply_10}}
    {inputName=Radius, input {expression=&self:MaterialExpressionScalarParameter_8}}
   ]
  }
  n15 = LinearInterpolate(
   B=Multiply(A=ScalarParameter() {defaultValue=4096, parameterName=Shape Dilation}, B=If(A=@n8, B=@n9.B, "A > B"=@n10, "A < B"=@n11))
   Alpha=OneMinus(None=ComponentMask(None=@n14) {b=true})
  ) {constB=0}
  n16 = MaterialFunctionCall() {materialFunction=&ReadCachedDistanceField}
  n17 = Saturate(None=Ceil(None=Add(A=@n15, B=@n16.World Distance))) {desc=Dilated Mask}
  LinearInterpolate(A=@n1.RGB, B=AppendVector(A=@n2.Velocity, B=LinearInterpolate(A=@n6, B=@n5, Alpha=@n7)), Alpha=@n17)
  n18 = MaterialFunctionCall(World Z=ScalarParameter() {parameterName=ZOffset}) {materialFunction=&LocalizeLandscapeZValue}
  n19 = Add(
   A=Multiply(
    A=Subtract(A=@n16.Water Height, B=@n18.Scale Only)
    B=CollectionParameter() {collection=&MPC_Landscape, parameterName=LandscapeZScale}
   )
   B=CollectionParameter() {collection=&MPC_Landscape, parameterName=LandscapeZLocation}
  )
  n20 = MaterialFunctionCall!(RG=TextureSampleParameter2D() {parameterName=LandscapeHeight, texture=&Black}.RGB) {materialFunction=/Landmass/Landscape/BlueprintBrushes/MF/UnpackRG8Height}
  n21 = CollectionParameter() {collection=&MPC_Landscape, parameterName=LandscapeZScale}
  n22 = NamedRerouteDeclaration(None=Add(
    A=Multiply(A=@n20.Result, B=@n21)
    B=CollectionParameter() {collection=&MPC_Landscape, parameterName=LandscapeZLocation}
   )) {name=Landscape Z, nodeColor=<rgba 0.3125 0 1 1>}
  n23 = Add(
   A=Multiply(A=Subtract(A=@n19, B=@n22) {desc=Water Depth}) {constB=-1}
   B=ScalarParameter() {defaultValue=32, parameterName=Water Dilation Threshold}
  )
  NamedRerouteDeclaration(None=Saturate(None=Ceil(None=@n23))) {name=Water Above Landscape Mask, nodeColor=<rgba 1 0.6328 0 1>}
  n24 = MaterialFunctionCall(World Z=ScalarParameter() {parameterName=ZOffset}) {
   desc="Z Offset pushes water but not terrain carving, thus affects Depth"
   materialFunction=&LocalizeLandscapeZValue
  }
  n25 = Divide(
   A=Subtract(A=@n2.Water Depth, B=@n24.Scale Only)
   B=Divide(
    A=ScalarParameter() {defaultValue=1000, parameterName=Max Encoded Depth}
    B=CollectionParameter() {collection=&MPC_Landscape, parameterName=LandscapeZScale}
   )
  ) {constB=1000}
  n26 = OneMinus(None=NamedRerouteDeclaration(None=Saturate(None=@n25)) {name=Normalized Water Depth, nodeColor=<rgba 0 0.1328 1 1>})
  n27 = TextureSampleParameter2D() {desc="RG=Velocity, B=Water Height,  A=WaterDepth", parameterName=CombinedVelocityAndHeight, texture=&Black}
  n28 = Min(A=@n26, B=@n27.A)
  n29 = Floor(None=@n27.A)
  OneMinus(None=LinearInterpolate(A=@n1.A, B=LinearInterpolate(A=@n28, B=@n26, Alpha=@n29), Alpha=@n17))
 ]
}
