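# Unlit material that finalizes a water texture for Water brushes from combined velocity/height and landscape height inputs.
# A Custom node box-blurs velocity over water texels only, and its neighbor flag forms a dilated water mask.
# Outside that mask, water height drops below terrain to hide it, then converts to world Z and joins velocity.
# It also computes an undilated mask and inverted terrain height normalized to MPC_Water's Z range.
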Material /Water/Materials/Brushes/FinalizeWaterTexture
graph {
 properties {
  bAllowNegativeEmissiveColor=true
  bUseTranslucencyVertexFog=false
  blendMode=BLEND_AlphaComposite
  pixelDepthOffsetMode=PDOM_Legacy
  shadingModel=MSM_Unlit
  usedShadingModels=Unlit
 }
 nodes [
  MaterialExpressionTextureObjectParameter_0 = TextureObjectParameter() {parameterName=CombinedVelocityAndHeight, texture=&Black}
  MaterialExpressionTextureCoordinate_0 = TextureCoordinate()
  n1 = CollectionParameter() {collection=/Landmass/Landscape/BlueprintBrushes/MPC/MPC_Landscape, parameterName=RTResX}
  n2 = CollectionParameter() {collection=&MPC_Landscape, parameterName=RTResY}
  MaterialExpressionMultiply_1 = Multiply(A=Divide(B=AppendVector(A=@n1, B=@n2)), B=ScalarParameter() {defaultValue=1, parameterName=Texel Scale})
  MaterialExpressionScalarParameter_4 = ScalarParameter() {defaultValue=3, parameterName=Radius}
  n3 = Custom!(
   Tex=@MaterialExpressionTextureObjectParameter_0
   UV=@MaterialExpressionTextureCoordinate_0
   TexelSize=@MaterialExpressionMultiply_1
   Radius=@MaterialExpressionScalarParameter_4
  ) {
   code="float2 accum = 0;\r\n\r\nfloat2 curUV = UV;\r\nfloat neighbors = 0;\r\n\r\nfor (int i = -Radius; i <= Radius; i++)\r\n{\r\n\r\n\tfor(int j=-Radius; j <= Radius; j++)\r\n\t{\r\n\t\t\r\n\t\tcurUV = UV + float2(TexelSize.x * (float)i, 0) + float2(0, TexelSize.y * (float)j);\r\n\t\tfloat4 cursample = Tex.SampleLevel(TexSampler, curUV, 0);\r\n\t\t\r\n\t\tfloat depthmask = 1 - floor(cursample.a);\r\n\t\t\r\n\t\tif (depthmask == 1.0)\r\n\t\t{\r\n\t\t\taccum += cursample.xy;\r\n\t\t\tneighbors += 1;\r\n\t\t\r\n\t\t}\r\n\t\t\r\n\t}\r\n\t\r\n}\r\n\r\nif(neighbors > 0)\r\n{\r\n\taccum.xy /= neighbors;\r\n}\r\n\r\nreturn float3( accum.xy, saturate(neighbors) );"
   description=Blur Velocities
   inputs [
    {inputName=Tex, input {expression=&self:MaterialExpressionTextureObjectParameter_0}}
    {inputName=UV, input {expression=&self:MaterialExpressionTextureCoordinate_0}}
    {inputName=TexelSize, input {expression=&self:MaterialExpressionMultiply_1}}
    {inputName=Radius, input {expression=&self:MaterialExpressionScalarParameter_4}}
   ]
  }
  n4 = TextureSampleParameter2D() {parameterName=CombinedVelocityAndHeight, texture=&Black}
  n5 = AppendVector(A=ComponentMask(None=@n3) {g=true, r=true}, B=@n4.B)
  n6 = ComponentMask(None=@n5) {b=true}
  n7 = MaterialFunctionCall!(RG=TextureSampleParameter2D() {parameterName=LandscapeHeight, texture=&Black}.RGB) {materialFunction=/Landmass/Landscape/BlueprintBrushes/MF/UnpackRG8Height}
  n8 = ScalarParameter() {defaultValue=256, parameterName=Water Terrain Offset}
  n9 = LinearInterpolate(
   A=@n6
   B=Min(A=@n6, B=Subtract(A=@n7.Result, B=@n8)) {desc="Away from Water Edge,  push water Under Terrain to Hide it"}
   Alpha=OneMinus(None=ComponentMask(None=@n3) {b=true, desc=Neighbor Check}) {desc=Dilated Water Mask}
  )
  n10 = CollectionParameter() {collection=&MPC_Landscape, parameterName=LandscapeZScale}
  AppendVector(
   A=ComponentMask(None=@n5) {g=true, r=true}
   B=Add(
    A=Multiply(A=@n9, B=@n10)
    B=CollectionParameter() {collection=&MPC_Landscape, parameterName=LandscapeZLocation}
   )
  )
  Floor(None=@n4.A) {desc="Un-Dilated Water Mask. 0 = Water"}
  n11 = CollectionParameter() {collection=&MPC_Landscape, parameterName=LandscapeZScale}
  n12 = CollectionParameter() {collection=/Water/Materials/MPC/MPC_Water, parameterName=TerrainMinZ}
  n13 = Subtract(
   A=Add(
    A=Multiply(A=@n7.Result, B=@n11)
    B=CollectionParameter() {collection=&MPC_Landscape, parameterName=LandscapeZLocation}
   )
   B=@n12
  )
  n14 = Subtract(A=CollectionParameter() {collection=&MPC_Water, parameterName=TerrainMaxZ}, B=@n12)
  OneMinus(None=Saturate(None=Divide(A=@n13, B=@n14)))
 ]
}
