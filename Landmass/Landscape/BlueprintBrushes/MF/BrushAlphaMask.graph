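# Landmass material function that computes the alpha mask for Blueprint landscape brushes from UV distance, Z height and displacement inputs.
# ZMaskParams optionally limits the offset distance with a signed or absolute height mask, built from HeightRT red minus Z height.
# The result is scaled by EdgeWidth, biased, displaced, saturated and multiplied by BrushOpacity.
# Separate UseZMask and WeightmapZRange parameters sit unconnected.
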MaterialFunction /Landmass/Landscape/BlueprintBrushes/MF/BrushAlphaMask
graph {nodes [
  n1 = Add(
   A=FunctionInput() {bUsePreviewValueAsDefault=true, inputName=UV Distance, inputType=FunctionInput_Scalar}
   B=ScalarParameter() {parameterName=EdgeCenterOffset}
  )
  n2 = VectorParameter() {desc="UseZMask, MaskLength, ZOffset, UseAbs", parameterName=ZMaskParams}
  n3 = Divide(
   A=Subtract(
    A=TextureSampleParameter2D() {parameterName=HeightRT, texture=/Engine/EngineResources/Black_Low.Black}.R
    B=FunctionInput() {bUsePreviewValueAsDefault=true, inputName=Z Height, inputType=FunctionInput_Scalar, sortPriority=3}
   )
   B=@n2.G
  )
  n4 = Multiply(A=@n2.G, B=@n2.G)
  n5 = Add(A=@n3, B=@n4)
  n6 = Add(A=Multiply(A=Abs(None=@n3)) {constB=-1}, B=@n4)
  n7 = LinearInterpolate(A=@n1, B=Min(A=@n1, B=LinearInterpolate(A=@n5, B=@n6, Alpha=@n2.A)), Alpha=@n2.R)
  n8 = ScalarParameter() {defaultValue=1, parameterName=EdgeWidth}
  n9 = Add(
   A=FunctionInput() {bUsePreviewValueAsDefault=true, inputName=Displacement, inputType=FunctionInput_Scalar, sortPriority=1}
   B=Add(A=Divide(A=@n7, B=@n8)) {constB=0.5}
  ) {desc=+ Texture}
  FunctionOutput(None=Multiply(A=Saturate(None=@n9), B=ScalarParameter() {defaultValue=1, parameterName=BrushOpacity})) {outputName=Alpha}
  ScalarParameter() {defaultValue=0.2, parameterName=WeightmapZRange}
  ScalarParameter() {defaultValue=1, parameterName=UseZMask}
 ]}
