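# Unlit Landmass brush material that shapes landscape height with an angled falloff around a cached mesh distance field and outputs packed RG8 height.
# Distance-based slopes, offset by masked texture displacement, form positive and negative shapes that smooth min/max against HeightRT.
# Blend modes and terracing follow.
# The result is then capped by depth decoded from CombinedAlphaAndHeightRT, faded in over a water intersect width.
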Material /Water/Materials/Brushes/MeshBrush_Angle
graph {
 properties {
  bAllowNegativeEmissiveColor=true
  floatPrecisionMode=MFPM_Full
  pixelDepthOffsetMode=PDOM_Legacy
  shadingModel=MSM_Unlit
  usedShadingModels=Unlit
 }
 nodes [
  n1 = ScalarParameter() {defaultValue=2, parameterName=T}
  MaterialFunctionCall!(Tiling=@n1) {materialFunction=/Landmass/Landscape/OldPrototype_BP/Mats/MF/CatmullRomTexture}
  n2 = TextureSampleParameter2D(UVs=Multiply(A=TextureCoordinate(), B=@n1)) {parameterName=BrushRT, texture=/Engine/MapTemplates/TilingNoise05}
  n3 = VectorParameter() {parameterName=DisplacementChannel, defaultValue=<rgba 1 0 0 0>}
  n4 = DotProduct(A=AppendVector(A=@n2.RGB, B=@n2.RGB), B=AppendVector(A=@n3.RGB, B=@n3.RGB))
  n5 = ScalarParameter() {defaultValue=0.5, parameterName=Displacement Midpoint}
  n6 = MaterialFunctionCall!(Cached DF, UV Offset) {materialFunction=/Landmass/Landscape/BlueprintBrushes/MF/ReadCachedDistanceField}
  n7 = LinearInterpolate(A=@n6.Terrain Height, B=@n6.Terrain Height, Alpha=ScalarParameter() {parameterName=UseMeshDepth})
  MaterialFunctionCall!(
   UV Distance=@n6.World Distance
   Displacement=Multiply(A=Subtract(A=@n4, B=@n5), B=ScalarParameter() {defaultValue=1, parameterName=WeightmapInfluence})
   Z Height=@n7
  ) {materialFunction=/Landmass/Landscape/BlueprintBrushes/MF/BrushAlphaMask}
  n8 = TextureSampleParameter2D() {parameterName=CombinedAlphaAndHeightRT, texture=/Engine/EngineResources/Black_Low.Black}
  n9 = CollectionParameter() {collection=/Landmass/Landscape/BlueprintBrushes/MPC/MPC_Landscape, parameterName=LandscapeZScale}
  n10 = Multiply(A=OneMinus(None=@n8.A), B=Divide(A=ScalarParameter() {defaultValue=1000, parameterName=Max Encoded Depth}, B=@n9)) {constB=1000, desc=De-Normalize}
  n11 = Subtract(A=Floor(None=@n8.B), B=@n10) {constB=0}
  n12 = ScalarParameter() {defaultValue=1, desc=Tangent, parameterName=FalloffTangent}
  n13 = CollectionParameter() {collection=&MPC_Landscape, parameterName=LandscapeZScale}
  n14 = Add(A=Multiply(A=@n6.World Distance, B=Divide(A=@n12, B=@n13)), B=Divide(A=ScalarParameter() {parameterName=EdgeCenterOffset}, B=@n13))
  n15 = LinearInterpolate(
   A=@n14
   B=MaterialFunctionCall!(A=@n14, B=Constant(), K=VectorParameter() {parameterName=SmoothingParams, defaultValue=<rgba 1 1 1 1>}.R) {materialFunction=/Landmass/Landscape/BlueprintBrushes/MF/SmoothIntersections}.Smooth Min
   Alpha=ScalarParameter() {parameterName=CapShapeInterior}
  )
  n16 = Subtract(A=@n4, B=@n5)
  n17 = ScalarParameter() {defaultValue=256, parameterName=BrushTexHeight}
  n18 = Add(
   A=@n7
   B=Multiply(
    A=MaterialFunctionCall!(UV Distance=@n6.World Distance) {materialFunction=/Landmass/Landscape/BlueprintBrushes/MF/BrushDisplaceMask}.Mask
    B=Multiply(A=@n16, B=@n17)
   ) {desc=Displacement * Mask}
  ) {desc=Height + Displacement}
  n19 = Add(A=Multiply(A=@n15) {constB=-1}, B=@n18) {desc=Negative Shape}
  n20 = VectorParameter() {parameterName=SmoothingParams, defaultValue=<rgba 1 1 1 1>}
  n21 = MaterialFunctionCall!(RG=ComponentMask(None=TextureSampleParameter2D() {parameterName=HeightRT, texture=&Black}.RGB) {g=true, r=true}) {materialFunction=/Landmass/Landscape/OldPrototype_BP/Mats/MF/UnpackRG8Height}
  n22 = MaterialFunctionCall!(AlphaBlend=@n21.Result, Min=@n21.Result, Max=@n21.Result, Additive=Constant()) {materialFunction=/Landmass/Landscape/BlueprintBrushes/MF/BrushBlendModes}
  n23 = Add(A=@n15, B=@n18) {desc=Positive Shape}
  n24 = Max(A=Subtract(A=@n18, B=@n22.Result) {desc=Intersection of Terrain and Brush}) {constB=0}
  n25 = MaterialFunctionCall(
   A=MaterialFunctionCall(A=@n22.Result, B=@n19, K=@n20.G) {materialFunction=&SmoothIntersections}.Smooth Min
   B=@n23
   K=LinearInterpolate(A=@n20.G, B=Min(A=@n24, B=@n20.G), Alpha=ScalarParameter() {parameterName=UseMeshDepth})
  ) {materialFunction=&SmoothIntersections}
  n26 = MaterialFunctionCall(RG=ComponentMask(None=TextureSampleParameter2D() {parameterName=HeightRT, texture=&Black}.RGB) {g=true, r=true}) {materialFunction=&UnpackRG8Height}
  n27 = MaterialFunctionCall!(
   Height=MaterialFunctionCall(
    AlphaBlend=@n25.Smooth Max
    Min=MaterialFunctionCall(A=@n22.Result, B=@n19, K=@n20.G) {desc=Smooth Min, materialFunction=&SmoothIntersections}.Smooth Min
    Max=MaterialFunctionCall(A=@n22.Result, B=@n23, K=@n20.G) {desc=Smooth Max, materialFunction=&SmoothIntersections}.Smooth Max
    Additive=Add(A=@n25.Smooth Max, B=@n26.Result)
   ) {materialFunction=&BrushBlendModes}.Result
   DF Mask=@n6.World Distance
  ) {materialFunction=/Landmass/Landscape/BlueprintBrushes/MF/BrushTerracing}
  n28 = ScalarParameter() {defaultValue=16, parameterName=Water Intersect Width}
  n29 = Saturate(None=Divide(A=@n10, B=Divide(A=@n28, B=@n9)) {constB=100})
  MaterialFunctionCall!(Height Int -32k to 32k=LinearInterpolate(A=Min(A=@n11, B=@n27.Result), B=@n27.Result, Alpha=OneMinus(None=@n29))) {materialFunction=/Landmass/Landscape/BlueprintBrushes/MF/PackRG8Height}
  ScalarParameter() {defaultValue=1394, parameterName=Z}
 ]
}
