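# Unlit full-precision brush material that adds a distance-field-shaped weight to a render target value, apparently for painting Water landscape weightmaps.
# It offsets the cached Landmass distance field by edge offset plus gradient width, then normalizes it by that width.
# A tiled noise texture, reduced to a selectable channel and centered on a midpoint, perturbs the gradient.
# The saturated result is scaled by opacity and added to the red channel of HeightRT.
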Material /Water/Materials/Brushes/MeshBrush_Weightmap
graph {
 properties {
  bAllowNegativeEmissiveColor=true
  floatPrecisionMode=MFPM_Full
  pixelDepthOffsetMode=PDOM_Legacy
  shadingModel=MSM_Unlit
  usedShadingModels=Unlit
 }
 nodes [
  n1 = MaterialFunctionCall!(Cached DF, UV Offset) {materialFunction=/Landmass/Landscape/BlueprintBrushes/MF/ReadCachedDistanceField}
  n2 = ScalarParameter() {defaultValue=4096, parameterName=GradientWidth}
  n3 = Add(A=ScalarParameter() {parameterName=EdgeOffset}, B=@n2)
  n4 = TextureSampleParameter2D(UVs=Multiply(A=TextureCoordinate(), B=ScalarParameter() {defaultValue=2, parameterName=T})) {parameterName=BrushRT, texture=/Engine/MapTemplates/TilingNoise05}
  n5 = VectorParameter() {parameterName=DisplacementChannel, defaultValue=<rgba 1 0 0 0>}
  n6 = DotProduct(A=AppendVector(A=@n4.RGB, B=@n4.RGB), B=AppendVector(A=@n5.RGB, B=@n5.RGB))
  n7 = ScalarParameter() {defaultValue=0.5, parameterName=Displacement Midpoint}
  n8 = Add(
   A=Divide(A=Add(A=@n1.World Distance, B=@n3), B=@n2)
   B=Multiply(A=Subtract(A=@n6, B=@n7), B=ScalarParameter() {defaultValue=1, parameterName=WeightmapInfluence})
  )
  Add(
   A=TextureSampleParameter2D() {parameterName=HeightRT, texture=&Black}.R
   B=Multiply(A=Saturate(None=@n8), B=ScalarParameter() {defaultValue=1, parameterName=Opacity})
  )
 ]
}
